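# Niagara effect for energy leaking from the cracked Zero Point containment sphere in the Dark Tower.
# Motes, shimmers, sparkles, mesh fragments and rotating spline meshes spawn on a sampled cracks mesh shifted by DTSphere_Offset, most pushed outward from that point, where two glow emitters also sit.
# System OpenAmount blends mote color toward User.InteriorColor, while DestabilizeAmount alters shimmer color and material.
# All but InnerGlow distance-cull and scale by distance.
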NiagaraSystem /Game/Effects/Fort_Effects/Effects/Environment/DarkTower/NS_DarkTower_ZPContain_Leak
dependencies {
 usedDataInterfaces [
  &NiagaraDataInterfaceCurve
  &NiagaraDataInterfaceStaticMesh
  &NiagaraDataInterfaceColorCurve
  &NiagaraDataInterfaceVectorCurve
 ]
 usedDynamicInputs []
 usedModules [
  &SystemState
  &self:SystemSpawnScript.SystemScriptSource.SystemScriptGraph.NiagaraNodeAssignment_0.SetVariables_#1
  &self:SystemSpawnScript.SystemScriptSource.SystemScriptGraph.NiagaraNodeAssignment_1.SetVariables_#2
  &self:SystemSpawnScript.SystemScriptSource.SystemScriptGraph.NiagaraNodeAssignment_2.SetVariables_#3
  &self:SystemSpawnScript.SystemScriptSource.SystemScriptGraph.NiagaraNodeAssignment_3.SetVariables_#4
 ]
}
emitters {
 InnerGlow {
  emitterUpdateScript [
   &EmitterState()
   &SpawnRate(SpawnRate=10, Use Spawn Probability=true)
  ]
  particleSpawnScript [
   /SetVariables(Particles.EndColor:LinearColor=&ColorFromCurve(
     ColorCurve=DataInterface() {propertyValues {
       AlphaCurve {keys [<at 0 1 RCIM_Cubic>,<at 1 1 RCIM_Cubic>]}
       BlueCurve {keys [<at 0 1 RCIM_Cubic>,<at 1 1 RCIM_Cubic>]}
       GreenCurve {keys [<at 0 0.1029 RCIM_Cubic>,<at 1 0.0955 RCIM_Cubic>]}
       RedCurve {keys [<at 0 0.3609 RCIM_Cubic>,<at 1 0.1117 RCIM_Cubic>]}
      }}
     CurveIndex=&RandomRangeFloat(Minimum=0, Maximum=1)
    )) {isSetParameters=true, module=SetVariables_#5}
   /Niagara/Modules/Spawn/Initialization/InitializeParticle(
    Color=&ColorFromCurve(
     ColorCurve=DataInterface() {propertyValues {
       AlphaCurve {keys [<at 0 1 RCIM_Cubic>,<at 1 1 RCIM_Cubic>]}
       BlueCurve {keys [<at 0 1 RCIM_Cubic>,<at 1 1 RCIM_Cubic>]}
       GreenCurve {keys [<at 0 0.7152 RCIM_Cubic>,<at 1 0.2609 RCIM_Cubic>]}
       RedCurve {keys [<at 0 0.1856 RCIM_Cubic>,<at 1 0.0709 RCIM_Cubic>]}
      }}
     CurveIndex=&RandomRangeFloat(Minimum=0, Maximum=1)
    )
    Lifetime=&RandomRangeFloat(Minimum=0.3, Maximum=0.75)
    Module.Write Color=true
    Module.Write Lifetime=true
    Module.Write Mass=true
    Module.Write Material Random=true
    Module.Write Position=false
    Module.Write Scale=false
    Module.Write SpriteRotation=false
    Module.Write SpriteSize=true
    Position=&SimulationPosition()
    Sprite Size=&Vector2DFromFloat(Value=800)
   )
   &SphereLocation(
    MaskSpawnGroup=false
    Offset=$System.DTSphere_Offset
    Override Seed=false
    Sphere Radius=0
    Surface Only=false
    UseNonUniformScale=false
    UseOffset=true
    UseSphereOrientation=false
   )
   &DynamicMaterialParameters(
    Index 0 Param 1=100
    Module.Param0WriteEnabled=true
    Module.Param1WriteEnabled=false
    Module.Param2WriteEnabled=false
    Module.Param3WriteEnabled=false
   )
   &CameraOffset(Camera Offset Amount=25)
  ]
  particleUpdateScript [
   &ParticleState()
   &ScaleColor(
    Scale Alpha=&FloatFromCurve(FloatCurve=DataInterface() {propertyValues {Curve {keys [{interpMode=RCIM_Cubic, tangentMode=RCTM_User, time=0, value=1}, <at 1 0>]}}})
    Scale RGB=&Multiply_VectorByFloat(
     Vector=<1 1 1>
     Float=&FloatFromCurve(
      FloatCurve=DataInterface() {propertyValues {Curve {keys [{interpMode=RCIM_Cubic, tangentMode=RCTM_User, time=0, value=0.2}, <at 1 1>]}}}
      Scale Curve=&RandomRangeFloat(Minimum=5, Maximum=200)
     )
    )
    ScaleA=true
    ScaleRGB=true
   )
   &ScaleSpriteSize(Scale Factor=&Vector2DFromFloat(Value=&FloatFromCurve(FloatCurve=DataInterface() {propertyValues {Curve {keys [{arriveTangent=3.2933, interpMode=RCIM_Cubic, leaveTangent=3.2933, tangentMode=RCTM_User, time=0, value=0.5}, <at 1 1>]}}})))
   /SetVariables(Particles.Initial.Color:LinearColor=&LerpLinearColors(
     StartColor=$Particles.Initial.Color
     EndColor=$Particles.EndColor
     LerpFactor=&FloatFromCurve(FloatCurve=DataInterface() {propertyValues {Curve {keys [
          {interpMode=RCIM_Cubic, tangentMode=RCTM_User, time=0, value=0}
          {arriveTangent=0.955, interpMode=RCIM_Cubic, leaveTangent=0.955, tangentMode=RCTM_User, time=1, value=1}
         ]}}})
    )) {isSetParameters=true, module=SetVariables_#6}
  ]
  rendererClasses [&NiagaraSpriteRendererProperties]
  renderers [{rendererClass=&NiagaraSpriteRendererProperties}]
 }
 Motes {
  emitterUpdateScript [
   &EmitterState(Enable Distance Culling=true, MaxDistance=7000, Module.bUseMaxDistance=true, Module.bUseMinDistance=false, Scalability Mode=::NewEnumerator1)
   &SpawnRate(SpawnRate=30, Use Spawn Probability=false)
  ]
  particleSpawnScript [
   /Niagara/Modules/Spawn/Initialization/V2/InitializeParticle(
    Color=&LerpLinearColors(
     StartColor=&RandomRangeLinearColor(Minimum=<rgba 2.0459 0.3129 3 1>, Maximum=<rgba 0.3154 4.3512 5 1>)
     EndColor=$User.InteriorColor
     LerpFactor=$System.OpenAmount
    )
    Lifetime Max=7
    Lifetime Min=3
    Lifetime Mode=::NewEnumerator1
    Mesh Scale Mode=::NewEnumerator4
    Position Mode=::NewEnumerator1
    Sprite Rotation Mode=::NewEnumerator2
    Sprite Size Mode=::NewEnumerator1
    Uniform Sprite Size Max=50
    Uniform Sprite Size Min=30
    UsePositionOffset=false
    Write Material Random=true
   )
   /Niagara/Modules/Spawn/MeshInterface/SampleStaticMesh(
    OverrideRandomSeed=false
    StaticMesh=DataInterface() {propertyValues {
      DefaultMesh=&/SM_DarkTower_ZPContainment_Cracks
      InstanceIndex=-1
      LODIndex=0
      PreviewMesh=null
      SoftSourceActor=null
      SourceMode=DefaultMeshOnly
      bAllowSamplingFromStreamingLODs=false
      bCaptureTransformsPerFrame=true
      bUsePhysicsBodyVelocity=false
      FilteredSockets []
      LODIndexUserParameter {parameter {name=None, typeDefHandle {registeredTypeIndex=108}, varData []}}
      MeshParameterBinding {parameter {name=None, typeDefHandle {registeredTypeIndex=1}, varData []}}
      SectionFilter {allowedMaterialSlots []}
     }}
   )
   /Niagara/Modules/Spawn/Location/StaticMeshLocation(
    Module.Enable Axis Constraint=false
    Normal Offset=5
    Sampled Mesh Position:NiagaraPosition=&Add_Vector(A=Linked() {linkedVariable {name=Particles.SampleStaticMesh.MeshPosition, type=NiagaraPosition}}, B=$System.DTSphere_Offset)
   )
   &AddVelocityFromPoint(
    Module.Use Falloff Distance=false
    Origin Offset=$System.DTSphere_Offset
    Velocity Strength=&FloatFromCurve(
     FloatCurve=DataInterface() {propertyValues {Curve {keys [
         {interpMode=RCIM_Cubic, tangentMode=RCTM_User, time=0, value=0.1}
         {arriveTangent=2.7242, interpMode=RCIM_Cubic, leaveTangent=2.7242, tangentMode=RCTM_User, time=1, value=1}
        ]}}}
     CurveIndex=&RandomRangeFloat(Minimum=0, Maximum=1)
     Scale Curve=250
    )
   )
   /SetVariables(Particles.Sparkle:NiagaraFloat=&UniformAOrBFloat(A=0, Distribution Weight=0.2)) {isSetParameters=true}
  ]
  particleUpdateScript [
   &ParticleState()
   &DynamicMaterialParameters(
    Index 0 Param 1=1
    Index 0 Param 2=1
    Index 0 Param 3=0.5
    Module.Param0WriteEnabled=true
    Module.Param1WriteEnabled=true
    Module.Param2WriteEnabled=true
    Module.Param3WriteEnabled=true
   )
   &ScaleColor(
    Scale Alpha=&FloatFromCurve(
     FloatCurve=DataInterface() {propertyValues {Curve {keys [
         <at 0 0 RCIM_Cubic>
         {interpMode=RCIM_Cubic, tangentMode=RCTM_User, time=0.1, value=0.9914}
         {interpMode=RCIM_Cubic, tangentMode=RCTM_User, time=0.7, value=1}
         <at 1 0>
        ]}}}
     Scale Curve=&FloatFromCurve(
      FloatCurve=DataInterface() {propertyValues {Curve {keys [
          <at 0 0.2>
          {arriveTangent=1.7182, interpMode=RCIM_Cubic, leaveTangent=1.7182, tangentMode=RCTM_User, time=0.1, value=0.2}
          {arriveTangent=0.5923, interpMode=RCIM_Cubic, leaveTangent=-0.0047, tangentMode=RCTM_Break, time=0.5, value=1}
          <at 1 1>
         ]}}}
      CurveIndex=$Emitter.Scalability.DistanceFraction
     )
    )
    ScaleA=true
    ScaleRGB=true
   )
   &ScaleSpriteSize(Scale Factor=&Vector2DFromFloat(Value=&FloatFromCurve(
      FloatCurve=DataInterface() {propertyValues {Curve {keys [
          <at 0 0 RCIM_Cubic>
          {arriveTangent=2.4312, interpMode=RCIM_Cubic, leaveTangent=2.4312, time=0.1483, value=1}
          {arriveTangent=-0.0468, interpMode=RCIM_Cubic, leaveTangent=-0.0468, time=0.2916, value=0.709}
          {arriveTangent=-0.2796, interpMode=RCIM_Cubic, leaveTangent=-0.2796, time=0.4899, value=0.984}
          {arriveTangent=0.7494, interpMode=RCIM_Cubic, leaveTangent=0.7494, tangentMode=RCTM_User, time=0.6851, value=0.599}
          <at 1 0>
         ]}}}
      Scale Curve=&FloatFromCurve(
       FloatCurve=DataInterface() {propertyValues {Curve {keys [
           <at 0 0.3>
           {arriveTangent=0.0785, interpMode=RCIM_Cubic, leaveTangent=0.0785, tangentMode=RCTM_User, time=0.2, value=0.3}
           {arriveTangent=0.5923, interpMode=RCIM_Cubic, leaveTangent=-0.0047, tangentMode=RCTM_Break, time=0.5, value=1}
           <at 1 1>
          ]}}}
       CurveIndex=$Emitter.Scalability.DistanceFraction
      )
     )))
   &VortexForce(
    Influence Falloff Radius=1600
    Origin Pull Amount=-0.5
    Use Influence Falloff=true
    Vortex Axis=<0 -1 0>
    Vortex Force Amount=100
    Vortex Origin:Vector3f=<1500 0 400>
   )
   &VectorNoiseForce(Module.Use Random Noise Time Interval=true)
   /SetVariables(Particles.SparkleSize:Vector2f=/Niagara/DynamicInputs/Add/Add_Vector2(
     A:Vector2f=$Particles.SpriteSize
     B:Vector2f=&Vector2DFromFloat(Value=&Multiply_Float(
       A=$Particles.Sparkle
       B=&FloatFromCurve(
        FloatCurve=DataInterface() {propertyValues {Curve {keys [
            <at 0.4 0.05 RCIM_Cubic>
            {arriveTangent=-4.4497, interpMode=RCIM_Cubic, leaveTangent=-4.4497, tangentMode=RCTM_User, time=0.45, value=1}
            <at 0.6 0.05 RCIM_Cubic>
           ]}}}
        Scale Curve=100
       )
      ))
    )) {isSetParameters=true}
   &Drag(Use Linear Drag=true, Use Rotational Drag=false)
   &GravityForce(Gravity=<0 0 -40>)
   &SolveForcesAndVelocity(Clamp Velocity=false, Limit Acceleration=false)
  ]
  rendererClasses [&NiagaraSpriteRendererProperties]
  renderers [
   {rendererClass=&NiagaraSpriteRendererProperties}
   {rendererClass=&NiagaraSpriteRendererProperties, rendererIndex=1}
  ]
 }
 OuterGlow {
  emitterUpdateScript [
   &EmitterState(Enable Distance Culling=true, MaxDistance=6000, Module.bUseMaxDistance=true, Module.bUseMinDistance=false, Scalability Mode=::NewEnumerator1)
   &SpawnRate(SpawnRate=1.5, Use Spawn Probability=true)
  ]
  particleSpawnScript [
   /SetVariables(Particles.EndColor:LinearColor=&ColorFromCurve(
     ColorCurve=DataInterface() {propertyValues {
       AlphaCurve {keys [<at 0 1 RCIM_Cubic>,<at 1 1 RCIM_Cubic>]}
       BlueCurve {keys [<at 0 1 RCIM_Cubic>,<at 1 1 RCIM_Cubic>]}
       GreenCurve {keys [<at 0 0.1029 RCIM_Cubic>,<at 1 0.0955 RCIM_Cubic>]}
       RedCurve {keys [<at 0 0.3609 RCIM_Cubic>,<at 1 0.1117 RCIM_Cubic>]}
      }}
     CurveIndex=&RandomRangeFloat(Minimum=0, Maximum=1)
    )) {isSetParameters=true, module=SetVariables_#5}
   /Niagara/Modules/Spawn/Initialization/InitializeParticle(
    Color=&ColorFromCurve(
     ColorCurve=DataInterface() {propertyValues {
       AlphaCurve {keys [<at 0 1 RCIM_Cubic>,<at 1 1 RCIM_Cubic>]}
       BlueCurve {keys [<at 0 1 RCIM_Cubic>,<at 1 1 RCIM_Cubic>]}
       GreenCurve {keys [<at 0 0.7152 RCIM_Cubic>,<at 1 0.2609 RCIM_Cubic>]}
       RedCurve {keys [<at 0 0.1856 RCIM_Cubic>,<at 1 0.0709 RCIM_Cubic>]}
      }}
     CurveIndex=&RandomRangeFloat(Minimum=0, Maximum=1)
    )
    Lifetime=&RandomRangeFloat(Minimum=1, Maximum=1.5)
    Module.Write Color=true
    Module.Write Lifetime=true
    Module.Write Mass=true
    Module.Write Material Random=true
    Module.Write Position=false
    Module.Write Scale=false
    Module.Write SpriteRotation=true
    Module.Write SpriteSize=true
    Position=&SimulationPosition()
    Sprite Rotation=&RandomRangeFloat(Minimum=0, Maximum=1)
    Sprite Size=&Multiply_Vector2(
     A=&RandomRangeVector2D(Minimum=<0.75 0.75>, Maximum=<1 1>)
     B=&Vector2DFromFloat(Value=1600)
    )
   )
   &SphereLocation(
    MaskSpawnGroup=false
    Offset=$System.DTSphere_Offset
    Override Seed=false
    Sphere Radius=0
    Surface Only=false
    UseNonUniformScale=false
    UseOffset=true
    UseSphereOrientation=false
   )
   &DynamicMaterialParameters(
    Index 0 Param 1=100
    Module.Param0WriteEnabled=true
    Module.Param1WriteEnabled=false
    Module.Param2WriteEnabled=false
    Module.Param3WriteEnabled=false
   )
   &CameraOffset(Camera Offset Amount=25)
  ]
  particleUpdateScript [
   &ParticleState()
   &ScaleColor(
    Scale Alpha=&FloatFromCurve(
     FloatCurve=DataInterface() {propertyValues {Curve {keys [{interpMode=RCIM_Cubic, tangentMode=RCTM_User, time=0, value=1}, <at 1 0>]}}}
     Scale Curve=&FloatFromCurve(
      FloatCurve=DataInterface() {propertyValues {Curve {keys [
          <at 0 0.03>
          {arriveTangent=0.0105, interpMode=RCIM_Cubic, leaveTangent=0.0105, tangentMode=RCTM_User, time=0.1, value=0.03}
          {arriveTangent=0.4347, interpMode=RCIM_Cubic, leaveTangent=-0.0047, tangentMode=RCTM_Break, time=0.5, value=0.1}
          <at 1 0.1>
         ]}}}
      CurveIndex=$Emitter.Scalability.DistanceFraction
     )
    )
    Scale RGB=&Multiply_VectorByFloat(
     Vector=<1 1 1>
     Float=&FloatFromCurve(
      FloatCurve=DataInterface() {propertyValues {Curve {keys [{interpMode=RCIM_Cubic, tangentMode=RCTM_User, time=0, value=0.2}, <at 1 1>]}}}
      Scale Curve=50
     )
    )
    ScaleA=true
    ScaleRGB=true
   )
   &ScaleSpriteSize(Scale Factor=&Vector2DFromFloat(Value=&FloatFromCurve(FloatCurve=DataInterface() {propertyValues {Curve {keys [{arriveTangent=1.4416, interpMode=RCIM_Cubic, leaveTangent=1.4416, tangentMode=RCTM_User, time=0, value=0.9}, <at 1 1>]}}})))
   /SetVariables(Particles.Initial.Color:LinearColor=&LerpLinearColors(
     StartColor=$Particles.Initial.Color
     EndColor=$Particles.EndColor
     LerpFactor=&FloatFromCurve(FloatCurve=DataInterface() {propertyValues {Curve {keys [
          {interpMode=RCIM_Cubic, tangentMode=RCTM_User, time=0, value=0}
          {arriveTangent=0.955, interpMode=RCIM_Cubic, leaveTangent=0.955, tangentMode=RCTM_User, time=1, value=1}
         ]}}})
    )) {isSetParameters=true, module=SetVariables_#6}
  ]
  rendererClasses [&NiagaraSpriteRendererProperties]
  renderers [{rendererClass=&NiagaraSpriteRendererProperties}]
 }
 Shimmers {
  emitterUpdateScript [
   &EmitterState(Enable Distance Culling=true, MaxDistance=7000, Module.bUseMaxDistance=true, Module.bUseMinDistance=false, Scalability Mode=::NewEnumerator1)
   &SpawnRate(SpawnRate=7, Use Spawn Probability=false)
  ]
  particleSpawnScript [
   /SetVariables(Particles.ParticleVelocity=250) {isSetParameters=true}
   /SetVariables(Particles.VelFromPointRand:NiagaraFloat=&FloatFromCurve(
     FloatCurve=DataInterface() {propertyValues {Curve {keys [
         {interpMode=RCIM_Cubic, tangentMode=RCTM_User, time=0, value=0.1}
         {arriveTangent=2.7242, interpMode=RCIM_Cubic, leaveTangent=2.7242, tangentMode=RCTM_User, time=1, value=1}
        ]}}}
     CurveIndex=&RandomRangeFloat(Minimum=0, Maximum=1)
     Scale Curve=$Particles.ParticleVelocity
    )) {isSetParameters=true}
   /Niagara/Modules/Spawn/Initialization/V2/InitializeParticle(
    Color=&LerpLinearColors(
     StartColor=$User.InteriorColor
     EndColor=<rgba 0.9714 0.5 1 0.25>
     LerpFactor=&Lerp_Float(A=0.7, B=1, Alpha=$System.DestabilizeAmount)
    )
    Lifetime Max=5
    Lifetime Min=4
    Lifetime Mode=::NewEnumerator1
    Sprite Rotation Mode=::NewEnumerator2
    Sprite Size Mode=::NewEnumerator1
    Uniform Sprite Size Max=400
    Uniform Sprite Size Min=300
    UsePositionOffset=false
    Write Material Random=true
   )
   &SampleStaticMesh(
    OverrideRandomSeed=false
    StaticMesh=DataInterface() {propertyValues {
      DefaultMesh=&SM_DarkTower_ZPContainment_Cracks
      InstanceIndex=-1
      LODIndex=0
      PreviewMesh=null
      SoftSourceActor=null
      SourceMode=DefaultMeshOnly
      bAllowSamplingFromStreamingLODs=false
      bCaptureTransformsPerFrame=true
      bUsePhysicsBodyVelocity=false
      FilteredSockets []
      LODIndexUserParameter {parameter {name=None, typeDefHandle {registeredTypeIndex=108}, varData []}}
      MeshParameterBinding {parameter {name=None, typeDefHandle {registeredTypeIndex=1}, varData []}}
      SectionFilter {allowedMaterialSlots []}
     }}
   )
   &StaticMeshLocation(
    Module.Enable Axis Constraint=false
    Sampled Mesh Position:NiagaraPosition=&Add_Vector(A=Linked() {linkedVariable {name=Particles.SampleStaticMesh.MeshPosition, type=NiagaraPosition}}, B=$System.DTSphere_Offset)
   )
   &AddVelocityFromPoint(Module.Use Falloff Distance=false, Origin Offset=$System.DTSphere_Offset, Velocity Strength=$Particles.VelFromPointRand)
   &CameraOffset(Camera Offset Amount=80)
  ]
  particleUpdateScript [
   &ParticleState()
   &DynamicMaterialParameters(
    Index 0 Param 1=$Particles.MaterialRandom
    Index 0 Param 2=&RandomRangeFloat(Minimum=0, Maximum=1)
    Index 0 Param 3=&Lerp_Float(A=0, B=2, Alpha=$System.DestabilizeAmount)
    Module.Param0WriteEnabled=true
    Module.Param1WriteEnabled=true
    Module.Param2WriteEnabled=true
    Module.Param3WriteEnabled=false
   )
   &Drag(Drag=0.4, Use Linear Drag=true, Use Rotational Drag=false)
   &GravityForce(Gravity=&RandomRangeVector(Minimum=<0 0 -4>, Maximum=<1 1 -20>))
   &ScaleColor(
    Scale Alpha=&FloatFromCurve(
     FloatCurve=DataInterface() {propertyValues {Curve {keys [<at 0 0 RCIM_Cubic>,<at 0.35 1 RCIM_Cubic>,<at 1 0>]}}}
     Scale Curve=&FloatFromCurve(
      FloatCurve=DataInterface() {propertyValues {Curve {keys [
          <at 0 0.3>
          {arriveTangent=1.7182, interpMode=RCIM_Cubic, leaveTangent=1.7182, tangentMode=RCTM_User, time=0.1, value=0.3}
          {arriveTangent=0.5923, interpMode=RCIM_Cubic, leaveTangent=-0.0047, tangentMode=RCTM_Break, time=0.5, value=1}
          <at 1 1>
         ]}}}
      CurveIndex=$Emitter.Scalability.DistanceFraction
     )
    )
    Scale RGB=&Multiply_VectorByFloat(
     Vector=<1 1 1>
     Float=&FloatFromCurve(
      FloatCurve=DataInterface() {propertyValues {Curve {keys [
          {arriveTangent=-8.1251, interpMode=RCIM_Cubic, leaveTangent=-8.1251, tangentMode=RCTM_User, time=0, value=1}
          {arriveTangent=-0.0389, interpMode=RCIM_Cubic, leaveTangent=-0.0388, tangentMode=RCTM_User, time=0.2, value=0.1}
          <at 1 0.1>
         ]}}}
      Scale Curve=10
     )
    )
    ScaleA=true
    ScaleRGB=true
   )
   &ScaleSpriteSize(Scale Factor=&Vector2DFromFloat(Value=&FloatFromCurve(
      FloatCurve=DataInterface() {propertyValues {Curve {keys [<at 0 0.7 RCIM_Cubic>, {arriveTangent=1.4874, interpMode=RCIM_Cubic, leaveTangent=1.4874, tangentMode=RCTM_User, time=1, value=1.5}]}}}
      Scale Curve=&Lerp_Float(
       A=1
       B=2.5
       Alpha=&NormalizeFloat(Value=$Particles.VelFromPointRand, Min=0, Max=$Particles.ParticleVelocity)
      )
     )))
   &SolveForcesAndVelocity(Clamp Velocity=false, Limit Acceleration=false)
  ]
  rendererClasses [&NiagaraSpriteRendererProperties]
  renderers [{rendererClass=&NiagaraSpriteRendererProperties}]
 }
 Sparkles_ambient {
  emitterUpdateScript [
   &EmitterState(Enable Distance Culling=true, MaxDistance=7000, Module.bUseMaxDistance=true, Module.bUseMinDistance=false, Scalability Mode=::NewEnumerator1)
   &SpawnRate(SpawnRate=50, Use Spawn Probability=false)
  ]
  particleSpawnScript [
   &InitializeParticle(
    Color Maximum=<rgba 0.1262 1.7405 2 1>
    Color Minimum=<rgba 1.3639 0.2086 2 1>
    Color Mode=::NewEnumerator2
    Lifetime Max=4
    Lifetime Min=2
    Lifetime Mode=::NewEnumerator1
    Mesh Scale Mode=::NewEnumerator4
    Position Mode=::NewEnumerator1
    Sprite Rotation Mode=::NewEnumerator2
    Sprite Size Mode=::NewEnumerator1
    Uniform Sprite Size Max=90
    Uniform Sprite Size Min=30
    UsePositionOffset=false
    Write Material Random=true
   )
   &SampleStaticMesh(
    OverrideRandomSeed=false
    StaticMesh=DataInterface() {propertyValues {
      DefaultMesh=&SM_DarkTower_ZPContainment_Cracks
      InstanceIndex=-1
      LODIndex=0
      PreviewMesh=null
      SoftSourceActor=null
      SourceMode=DefaultMeshOnly
      bAllowSamplingFromStreamingLODs=false
      bCaptureTransformsPerFrame=true
      bUsePhysicsBodyVelocity=false
      FilteredSockets []
      LODIndexUserParameter {parameter {name=None, typeDefHandle {registeredTypeIndex=108}, varData []}}
      MeshParameterBinding {parameter {name=None, typeDefHandle {registeredTypeIndex=1}, varData []}}
      SectionFilter {allowedMaterialSlots []}
     }}
   )
   &StaticMeshLocation(
    Module.Enable Axis Constraint=false
    Normal Offset=5
    Sampled Mesh Position:NiagaraPosition=&Add_Vector(A=Linked() {linkedVariable {name=Particles.SampleStaticMesh.MeshPosition, type=NiagaraPosition}}, B=$System.DTSphere_Offset)
   )
   &AddVelocityFromPoint(
    Module.Use Falloff Distance=false
    Origin Offset=$System.DTSphere_Offset
    Velocity Strength=&FloatFromCurve(
     FloatCurve=DataInterface() {propertyValues {Curve {keys [
         {interpMode=RCIM_Cubic, tangentMode=RCTM_User, time=0, value=0.1}
         {arriveTangent=2.7242, interpMode=RCIM_Cubic, leaveTangent=2.7242, tangentMode=RCTM_User, time=1, value=1}
        ]}}}
     CurveIndex=&RandomRangeFloat(Minimum=0, Maximum=1)
     Scale Curve=450
    )
   )
  ]
  particleUpdateScript [
   &ParticleState()
   &Drag(Drag=0.5, Use Linear Drag=true, Use Rotational Drag=false)
   &CurlNoiseForce(
    Noise Strength=&FloatFromCurve(
     FloatCurve=DataInterface() {propertyValues {Curve {keys [
         {interpMode=RCIM_Cubic, tangentMode=RCTM_User, time=0, value=0}
         {arriveTangent=0.8922, interpMode=RCIM_Cubic, leaveTangent=0.8922, tangentMode=RCTM_User, time=1, value=1}
        ]}}}
     Scale Curve=500
    )
    Pan Noise=false
   )
   &DynamicMaterialParameters(
    Index 0 Param 1=1
    Index 0 Param 2=1
    Index 0 Param 3=0.5
    Module.Param0WriteEnabled=true
    Module.Param1WriteEnabled=true
    Module.Param2WriteEnabled=true
    Module.Param3WriteEnabled=true
   )
   &ScaleColor(
    Scale Alpha=&FloatFromCurve(
     FloatCurve=DataInterface() {propertyValues {Curve {keys [
         {arriveTangent=27.7425, interpMode=RCIM_Cubic, leaveTangent=27.7426, tangentMode=RCTM_User, time=0, value=0}
         {interpMode=RCIM_Cubic, tangentMode=RCTM_User, time=0.1, value=1}
         {arriveTangent=-2.0835, interpMode=RCIM_Cubic, leaveTangent=-2.0835, tangentMode=RCTM_User, time=0.8081, value=0.5934}
         <at 1 0>
        ]}}}
     Scale Curve=&FloatFromCurve(
      FloatCurve=DataInterface() {propertyValues {Curve {keys [
          <at 0 0.1>
          {arriveTangent=1.7182, interpMode=RCIM_Cubic, leaveTangent=1.7182, tangentMode=RCTM_User, time=0.1, value=0.1}
          {arriveTangent=0.5923, interpMode=RCIM_Cubic, leaveTangent=-0.0047, tangentMode=RCTM_Break, time=0.5, value=1}
          <at 1 1>
         ]}}}
      CurveIndex=$Emitter.Scalability.DistanceFraction
     )
    )
    Scale RGB=&VectorFromFloat(Value=&FloatFromCurve(
      FloatCurve=DataInterface() {propertyValues {Curve {keys [
          {arriveTangent=0.0361, interpMode=RCIM_Cubic, leaveTangent=0.0361, tangentMode=RCTM_User, time=0, value=1}
          {arriveTangent=-0.2536, interpMode=RCIM_Cubic, leaveTangent=-0.2536, tangentMode=RCTM_User, time=0.2, value=0.1}
          <at 1 0.05>
         ]}}}
      Scale Curve=25
     ))
    ScaleA=true
    ScaleRGB=true
   )
   &ScaleSpriteSize(Scale Factor=&Vector2DFromFloat(Value=&FloatFromCurve(
      FloatCurve=DataInterface() {propertyValues {Curve {keys [
          <at 0 0 RCIM_Cubic>
          {arriveTangent=0.836, interpMode=RCIM_Cubic, leaveTangent=0.836, time=0.2, value=1}
          {arriveTangent=-1.8455, interpMode=RCIM_Cubic, leaveTangent=-1.8455, tangentMode=RCTM_User, time=0.6507, value=0.544}
          <at 1 0>
         ]}}}
      Scale Curve=&FloatFromCurve(
       FloatCurve=DataInterface() {propertyValues {Curve {keys [
           <at 0 0.3>
           {arriveTangent=0.0785, interpMode=RCIM_Cubic, leaveTangent=0.0785, tangentMode=RCTM_User, time=0.2, value=0.3}
           {arriveTangent=0.5923, interpMode=RCIM_Cubic, leaveTangent=-0.0047, tangentMode=RCTM_Break, time=0.5, value=1}
           <at 1 1>
          ]}}}
       CurveIndex=$Emitter.Scalability.DistanceFraction
      )
     )))
   &SolveForcesAndVelocity(Clamp Velocity=false, Limit Acceleration=false)
  ]
  rendererClasses [&NiagaraSpriteRendererProperties]
  renderers [{rendererClass=&NiagaraSpriteRendererProperties}]
 }
 SpinSplines {
  emitterUpdateScript [
   &EmitterState(Enable Distance Culling=true, MaxDistance=6000, Module.bUseMaxDistance=true, Module.bUseMinDistance=false, Scalability Mode=::NewEnumerator1)
   &SpawnRate(SpawnRate=5, Use Spawn Probability=false)
  ]
  particleSpawnScript [
   /SetVariables(Particles.ParticleRandom:NiagaraFloat=&RandomRangeFloat(Minimum=-1, Maximum=1)) {isSetParameters=true}
   /Niagara/Modules/Spawn/Initialization/InitializeParticle(
    Color=&LerpLinearColors(
     StartColor=<rgba 0.7057 0.0918 2 0.25>
     EndColor=<rgba 0.4375 0.6164 2 0.3>
     LerpFactor=&RandomRangeFloat(Minimum=0, Maximum=1)
    )
    Lifetime=&RandomRangeFloat(Minimum=4.5, Maximum=6)
    Mesh Scale=&RandomRangeVector(Minimum=<1.25 1.25 1.25>, Maximum=<3 1.5 1.4>)
    Module.Write Color=true
    Module.Write Lifetime=true
    Module.Write Mass=true
    Module.Write Material Random=true
    Module.Write Position=false
    Module.Write Scale=true
    Module.Write SpriteRotation=false
    Module.Write SpriteSize=false
    Position=&MakeVector(
     X=&RandomRangeFloat(Minimum=0, Maximum=50)
     Y=&Multiply_Float(
      A=1
      B=&FloatFromCurve(
       FloatCurve=DataInterface() {propertyValues {Curve {keys [<at -1 -0.3>,<at -0.01 -0.1 RCIM_Constant>,<at 0.01 0.7>,<at 1 1>]}}}
       CurveIndex=$Particles.ParticleRandom
       Scale Curve=350
      )
     )
     Z=&RandomRangeFloat(Minimum=450, Maximum=600)
    )
    Sprite Size=<5 5>
   )
   &SampleStaticMesh(
    OverrideRandomSeed=false
    StaticMesh=DataInterface() {propertyValues {
      DefaultMesh=&SM_DarkTower_ZPContainment_Cracks
      InstanceIndex=-1
      LODIndex=0
      PreviewMesh=null
      SoftSourceActor=null
      SourceMode=DefaultMeshOnly
      bAllowSamplingFromStreamingLODs=false
      bCaptureTransformsPerFrame=true
      bUsePhysicsBodyVelocity=false
      FilteredSockets []
      LODIndexUserParameter {parameter {name=None, typeDefHandle {registeredTypeIndex=108}, varData []}}
      MeshParameterBinding {parameter {name=None, typeDefHandle {registeredTypeIndex=1}, varData []}}
      SectionFilter {allowedMaterialSlots []}
     }}
   )
   &StaticMeshLocation(
    Module.Enable Axis Constraint=false
    Sampled Mesh Position:NiagaraPosition=&Add_Vector(A=Linked() {linkedVariable {name=Particles.SampleStaticMesh.MeshPosition, type=NiagaraPosition}}, B=$System.DTSphere_Offset)
   )
   &InitialMeshOrientation(
    Orientation Vector=$Particles.SampleStaticMesh.MeshNormal
    Rotation=&MakeVector(
     X=&RandomRangeFloat(Minimum=0.1, Maximum=0.2)
     Y=&FloatFromCurve(
      FloatCurve=DataInterface() {propertyValues {Curve {keys [{arriveTangent=2.6536, interpMode=RCIM_Cubic, time=-1, value=1}, {interpMode=RCIM_Cubic, leaveTangent=2.6119, time=1, value=0.5}]}}}
      CurveIndex=$Particles.ParticleRandom
      Scale Curve=0.25
     )
     Z=&FloatFromCurve(
      FloatCurve=DataInterface() {propertyValues {Curve {keys [{arriveTangent=2.6536, interpMode=RCIM_Cubic, time=-1, value=1}, {interpMode=RCIM_Cubic, leaveTangent=2.6119, time=1, value=0.5}]}}}
      CurveIndex=$Particles.ParticleRandom
      Scale Curve=0.25
     )
    )
    Use Orientation Vector=true
    Use Rotation Vector=true
   ) {module=InitialMeshOrientation001}
   &AddVelocity(
    Scale Added Velocity=&VectorFromFloat(Value=50)
    Velocity=&MakeVector(X=0, Y=0, Z=0)
   ) {enabled=false}
  ]
  particleUpdateScript [
   &ParticleState()
   &ScaleColor(
    Scale Alpha=&FloatFromCurve(
     FloatCurve=DataInterface() {propertyValues {Curve {keys [
         <at 0 1 RCIM_Cubic>
         {arriveTangent=0.5714, interpMode=RCIM_Cubic, leaveTangent=0.5714, time=0.1, value=2}
         {arriveTangent=-2.7491, interpMode=RCIM_Cubic, leaveTangent=-2.7491, tangentMode=RCTM_User, time=0.7, value=1.4}
         <at 1 0.75 RCIM_Cubic>
        ]}}}
     Scale Curve=&FloatFromCurve(
      FloatCurve=DataInterface() {propertyValues {Curve {keys [
          <at 0 0.2>
          {arriveTangent=1.7182, interpMode=RCIM_Cubic, leaveTangent=1.7182, tangentMode=RCTM_User, time=0.1, value=0.2}
          {arriveTangent=0.5923, interpMode=RCIM_Cubic, leaveTangent=-0.0047, tangentMode=RCTM_Break, time=0.5, value=1}
          <at 1 1>
         ]}}}
      CurveIndex=$Emitter.Scalability.DistanceFraction
     )
    )
    Scale RGB=&Multiply_VectorByFloat(
     Vector=<1 1 1>
     Float=&FloatFromCurve(
      FloatCurve=DataInterface() {propertyValues {Curve {keys [
          {arriveTangent=0.103, interpMode=RCIM_Cubic, leaveTangent=0.103, tangentMode=RCTM_User, time=0, value=4}
          {interpMode=RCIM_Cubic, tangentMode=RCTM_User, time=0.3, value=1.1}
          <at 1 1>
         ]}}}
      Scale Curve=2
     )
    )
    ScaleA=true
    ScaleRGB=true
   )
   /SetVariables(Particles.Position=&MakeVector(
     X=&MakeFloatFromVector(Vector=Linked() {linkedVariable {name=Particles.Position, type=NiagaraPosition}}, Channel=::NewEnumerator0)
     Y=&MakeFloatFromVector(Vector=Linked() {linkedVariable {name=Particles.Position, type=NiagaraPosition}}, Channel=::NewEnumerator1)
     Z=&Lerp_Float(
      A=700
      B=&MakeFloatFromVector(Vector=Linked() {linkedVariable {name=Particles.Initial.Position, type=NiagaraPosition}}, Channel=::NewEnumerator2)
      Alpha=&FloatFromCurve(FloatCurve=DataInterface() {propertyValues {Curve {keys [
           <at 0 1 RCIM_Cubic>
           {arriveTangent=-1.8888, interpMode=RCIM_Cubic, leaveTangent=-1.8888, tangentMode=RCTM_User, time=0.6201, value=0.4065}
           <at 1 0>
          ]}}})
     )
    )) {isSetParameters=true}
   &DynamicMaterialParameters(
    Index 0 Param 1=&FloatFromCurve(FloatCurve=DataInterface() {propertyValues {Curve {keys [
         {arriveTangent=6.1791, interpMode=RCIM_Cubic, leaveTangent=6.1791, tangentMode=RCTM_User, time=0, value=0}
         {interpMode=RCIM_Cubic, tangentMode=RCTM_User, time=0.1943, value=1}
         {interpMode=RCIM_Cubic, tangentMode=RCTM_User, time=0.75, value=1}
         <at 1 0 RCIM_Cubic>
        ]}}})
    Index 0 Param 2=&FloatFromCurve(
     FloatCurve=DataInterface() {propertyValues {Curve {keys [
         {arriveTangent=9.0034, interpMode=RCIM_Cubic, leaveTangent=9.0034, tangentMode=RCTM_User, time=0, value=0}
         {interpMode=RCIM_Cubic, tangentMode=RCTM_User, time=0.2, value=1}
         {interpMode=RCIM_Cubic, tangentMode=RCTM_User, time=0.75, value=1}
         {arriveTangent=-7.1709, interpMode=RCIM_Cubic, leaveTangent=-7.1709, tangentMode=RCTM_User, time=1, value=0}
        ]}}}
     Scale Curve=0.7
    )
    Index 0 Param 3=&FloatFromCurve(FloatCurve=DataInterface() {propertyValues {Curve {keys [<at 0 1>,<at 1 0>]}}}, Scale Curve=5)
    Index 0 Param 4=&RandomRangeFloat(Minimum=0, Maximum=1)
    Module.Param0WriteEnabled=true
    Module.Param1WriteEnabled=true
    Module.Param2WriteEnabled=true
    Module.Param3WriteEnabled=true
   )
   &UpdateMeshOrientation(
    Rotation Coordinate Space=::NewEnumerator3
    Rotation Rate=&FloatFromCurve(
     FloatCurve=DataInterface() {propertyValues {Curve {keys [
         {arriveTangent=7.0045, interpMode=RCIM_Cubic, time=0, value=0.1}
         {arriveTangent=-0.1, interpMode=RCIM_Cubic, leaveTangent=-0.1, time=0.6, value=1}
         <at 1 0 RCIM_Cubic>
        ]}}}
     Scale Curve=&RandomRangeFloat(Minimum=0.04, Maximum=0.06)
    )
    Rotation Vector=<1 0 0>
   ) {module=UpdateMeshOrientation001}
   &ScaleMeshSize(Scale Factor=&VectorFromCurve(VectorCurve=DataInterface() {propertyValues {
       XCurve {keys [
         {arriveTangent=1.2655, interpMode=RCIM_Cubic, leaveTangent=0.212, tangentMode=RCTM_User, time=0, value=0.45}
         {arriveTangent=1.2684, interpMode=RCIM_Cubic, leaveTangent=1.2684, tangentMode=RCTM_User, time=1, value=1}
        ]}
       YCurve {keys [
         {arriveTangent=-0.4592, interpMode=RCIM_Cubic, time=0.1, value=1}
         {arriveTangent=-0.5381, interpMode=RCIM_Cubic, leaveTangent=-1.7444, tangentMode=RCTM_User, time=1, value=0.6}
        ]}
       ZCurve {keys [
         {arriveTangent=0.674, interpMode=RCIM_Cubic, leaveTangent=0.674, tangentMode=RCTM_User, time=0, value=0.25}
         {interpMode=RCIM_Cubic, leaveTangent=1.1784, time=1, value=1}
        ]}
      }}))
   &Drag(Use Linear Drag=true, Use Rotational Drag=false)
   &SolveForcesAndVelocity(Clamp Velocity=false, Limit Acceleration=false)
  ]
  rendererClasses [&NiagaraMeshRendererProperties]
  renderers [{rendererClass=&NiagaraMeshRendererProperties, rendererIndex=0}]
 }
 TinyFrags {
  emitterUpdateScript [
   &EmitterState(Enable Distance Culling=true, MaxDistance=7000, Module.bUseMaxDistance=true, Module.bUseMinDistance=false, Scalability Mode=::NewEnumerator1)
   &SpawnRate(SpawnRate=10, Use Spawn Probability=false)
  ]
  particleSpawnScript [
   /SetVariables(Particles.MeshOrientation=&MakeQuatFromAxisAngle(
     Axis=&RandomRangeVector(Minimum=<0 0 0>, Maximum=<1 1 1>, Evaluation Type=::NewEnumerator1)
     AngleInDegrees=90
    )) {isSetParameters=true}
   /SetVariables(Particles.SpawnRotationVector=&RandomRangeVector(Minimum=<-0.15 -0.15 -0.15>, Maximum=<0.15 0.15 0.15>, Evaluation Type=::NewEnumerator1)) {isSetParameters=true}
   &InitializeParticle(
    Lifetime=&RandomRangeFloat(Minimum=3, Maximum=4, Evaluation Type=::NewEnumerator1)
    Mesh Scale=&Multiply_VectorByFloat(Vector=&RandomRangeVector(Minimum=<4 4 5>, Maximum=<7 7 7>, Evaluation Type=::NewEnumerator1), Float=1)
    Module.Write Color=true
    Module.Write Lifetime=true
    Module.Write Mass=true
    Module.Write Material Random=true
    Module.Write Position=false
    Module.Write Scale=true
    Module.Write SpriteRotation=false
    Module.Write SpriteSize=false
   )
   &SampleStaticMesh(
    OverrideRandomSeed=false
    StaticMesh=DataInterface() {propertyValues {
      DefaultMesh=&SM_DarkTower_ZPContainment_Cracks
      InstanceIndex=-1
      LODIndex=0
      PreviewMesh=null
      SoftSourceActor=null
      SourceMode=DefaultMeshOnly
      bAllowSamplingFromStreamingLODs=false
      bCaptureTransformsPerFrame=true
      bUsePhysicsBodyVelocity=false
      FilteredSockets []
      LODIndexUserParameter {parameter {name=None, typeDefHandle {registeredTypeIndex=108}, varData []}}
      MeshParameterBinding {parameter {name=None, typeDefHandle {registeredTypeIndex=1}, varData []}}
      SectionFilter {allowedMaterialSlots []}
     }}
   )
   &StaticMeshLocation(
    Module.Enable Axis Constraint=false
    Normal Offset=5
    Sampled Mesh Position:NiagaraPosition=&Add_Vector(A=Linked() {linkedVariable {name=Particles.SampleStaticMesh.MeshPosition, type=NiagaraPosition}}, B=$System.DTSphere_Offset)
   )
   &Color(Color=&MakeLinearColorFromVectorAndFloat("Vector (RGB)"=<9.3572 4.2143 10>, "Float (Alpha)"=&ClampFloat(Float=1)))
   &AddVelocityFromPoint(
    Module.Use Falloff Distance=false
    Origin Offset=$System.DTSphere_Offset
    Velocity Strength=&FloatFromCurve(
     FloatCurve=DataInterface() {propertyValues {Curve {keys [
         {interpMode=RCIM_Cubic, tangentMode=RCTM_User, time=0, value=0.1}
         {arriveTangent=2.7242, interpMode=RCIM_Cubic, leaveTangent=2.7242, tangentMode=RCTM_User, time=1, value=1}
        ]}}}
     CurveIndex=&RandomRangeFloat(Minimum=0, Maximum=1)
     Scale Curve=450
    )
   )
  ]
  particleUpdateScript [
   &ParticleState()
   &Drag(Drag=0.2, Use Linear Drag=true, Use Rotational Drag=false)
   &SolveForcesAndVelocity(Clamp Velocity=false, Limit Acceleration=false)
   &Color(
    Color=&HueShiftLinearColor(
     Hue=0
     Saturation=2
     Value=&FloatFromCurve(
      FloatCurve=DataInterface() {propertyValues {Curve {keys [
          {interpMode=RCIM_Cubic, tangentMode=RCTM_User, time=0, value=1}
          {arriveTangent=-0.9, interpMode=RCIM_Cubic, leaveTangent=-0.9, tangentMode=RCTM_User, time=0.3, value=0.3}
          <at 1 0.1>
         ]}}}
      Scale Curve=10
     )
     Linear Color=&MultiplyLinearColor(
      Linear Color=&MultiplyLinearColor(
       Linear Color=&ColorFromCurve(ColorCurve=DataInterface() {propertyValues {
          AlphaCurve {keys [<at 0 0 RCIM_Cubic>,<at 0.1603 0.25>,<at 0.5248 0.5 RCIM_Cubic>,<at 0.8084 0.25>,<at 1 0>]}
          BlueCurve {keys [<at 0 1 RCIM_Cubic>,<at 0.254 1>,<at 0.519 1>,<at 1 1 RCIM_Cubic>]}
          GreenCurve {keys [<at 0 0.4592 RCIM_Cubic>,<at 0.254 0.3343>,<at 0.519 0>,<at 1 0.729 RCIM_Cubic>]}
          RedCurve {keys [<at 0 0.1944 RCIM_Cubic>,<at 0.254 0.0562>,<at 0.519 0.2449>,<at 1 0.7623 RCIM_Cubic>]}
         }})
       Scale Factor=&MakeVector4_FromFloat(Float=$System.GlobalDistScale)
      )
      Scale Factor=<1 1 1 1>
     )
    )
    Scale Alpha=&FloatFromCurve(FloatCurve=DataInterface() {propertyValues {Curve {keys [
         {interpMode=RCIM_Cubic, tangentMode=RCTM_User, time=0, value=0}
         {arriveTangent=-0.0305, interpMode=RCIM_Cubic, leaveTangent=-0.0305, tangentMode=RCTM_User, time=0.2166, value=1}
         <at 1 0>
        ]}}})
   ) {module=Color001}
   &UpdateMeshOrientation(Rotation Vector=$Particles.SpawnRotationVector) {module=UpdateMeshOrientation001}
   &ScaleMeshSize(Scale Factor=&VectorFromFloat(Value=&FloatFromCurve(
      FloatCurve=DataInterface() {propertyValues {Curve {keys [<at 0 0 RCIM_Cubic>, <at 0.2 1>, {arriveTangent=-1.25, interpMode=RCIM_Cubic, leaveTangent=-1.25, time=0.8, value=1}, <at 1 0>]}}}
      Scale Curve=&FloatFromCurve(
       FloatCurve=DataInterface() {propertyValues {Curve {keys [{interpMode=RCIM_Cubic, tangentMode=RCTM_Break, time=0, value=1}, <at 60000 8>, <at 70000 0>]}}}
       CurveIndex=$Engine.Owner.LODDistance
       Scale Curve=&FloatFromCurve(
        FloatCurve=DataInterface() {propertyValues {Curve {keys [
            <at 0 0.5>
            {arriveTangent=0.0785, interpMode=RCIM_Cubic, leaveTangent=0.0785, tangentMode=RCTM_User, time=0.2, value=0.5}
            {arriveTangent=0.5923, interpMode=RCIM_Cubic, leaveTangent=-0.0047, tangentMode=RCTM_Break, time=0.5, value=1}
            <at 1 1>
           ]}}}
        CurveIndex=$Emitter.Scalability.DistanceFraction
       )
      )
     )))
  ]
  rendererClasses [&NiagaraMeshRendererProperties]
  renderers [{rendererClass=&NiagaraMeshRendererProperties, rendererIndex=0}]
 }
}
summary {
 systemName=NS_DarkTower_ZPContain_Leak
 emitters [
  {emitterName=Shimmers, rendererClasses [&NiagaraSpriteRendererProperties]}
  {emitterName=Motes, rendererClasses [&NiagaraSpriteRendererProperties]}
  {emitterName=Sparkles_ambient, rendererClasses [&NiagaraSpriteRendererProperties]}
  {emitterName=InnerGlow, rendererClasses [&NiagaraSpriteRendererProperties]}
  {emitterName=OuterGlow, rendererClasses [&NiagaraSpriteRendererProperties]}
  {emitterName=TinyFrags, rendererClasses [&NiagaraMeshRendererProperties]}
  {emitterName=SpinSplines, rendererClasses [&NiagaraMeshRendererProperties]}
 ]
 userVariables [
  User.InteriorColor {type=LinearColor, description="", defaultValue {struct=LinearColor, value=<rgba 1 1 1 1>}}
  User.OpenAmount {type=NiagaraFloat, defaultValue=0, description=""}
  User.ZeroPointPosition {type=Vector3f, description="", defaultValue {struct=Vector3f, value=<6976 5908 8064>}}
 ]
}
